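# Anim Blueprint for the MimicGlean backpack on the Scholar pack skeleton, playing reactive open/close animations.
# Frontend and gameplay state machines are selected by bool, layered with a curve-driven floating additive, then pack and root bones are offset and scaled.
# Delay-pulsed IsOpen/FullyOpen flags follow enemy eliminations or a repeating frontend timer, initially started if the WordUnity modifier enables reactivity.
# Anim notifies broadcast ReactivityCue, spawn WordUnity Niagara FX and shut it off.
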AnimBlueprint /BRCosmetics/Accessories/FORT_Backpacks/Backpack_M_MED_MimicGlean/Meshes/F_MED_Scholar_MimicGlean_Pack_AnimBP
target_skeleton=/Game/Accessories/FORT_Backpacks/Backpack_F_MED_Scholar/Meshes/F_MED_Scholar_Pack_Skeleton
graphs [
 AnimGraph {nodes [
   AnimGraphNode_StateMachine_0 = FE_Pack() {is=title AnimGraphNode_StateMachine}
   AnimGraphNode_StateMachine_2 = Pack() {is=title AnimGraphNode_StateMachine}
   n1 = Apply Additive!(
    Base=Blend Poses by bool(
     BlendPose_0=@AnimGraphNode_StateMachine_0
     BlendPose_1=@AnimGraphNode_StateMachine_2
     bActiveValue=Get bShouldDisplayFrontendReactivity() -> (bShouldDisplayFrontendReactivity:bool)
    )
    Additive="Sequence Player (additive) 'Idle_Open_Float_PROP'"() {sequence=/Game/Accessories/FORT_Backpacks/Backpack_F_MED_Scholar/Animation/Idle_Open_Float_PROP}
    AlphaCurveName:Name=Floating
   ) -> (Pose) {alphaInputType=Curve, alphaBoolBlend {blendOption=Linear}}
   n2 = "Transform (Modify) Bone - Bone: root"(
    ComponentPose="Transform (Modify) Bone - Bone: pack"(ComponentPose=Local To Component(LocalPose=@n1), Translation="-2.5,0,0", Rotation="0,0,0") {translationMode=BMM_Additive, translationSpace=BCS_BoneSpace, boneToModify {boneName=pack}, translation {x=-2.5}}
    Translation="0,0,0"
    Rotation="0,0,0"
    Scale="0.9,0.9,0.9"
   ) {
    scaleMode=BMM_Replace
    scaleSpace=BCS_ParentBoneSpace
    alphaBoolBlend {blendOption=Linear}
    boneToModify {boneName=root}
    scale=<0.9 0.9 0.9>
   }
   Output Pose(Result=Component To Local(ComponentPose=@n2)) {layerGroup=None}
  ]}
 FE_Pack {
  entry=FE_Closed
  path=AnimGraph.AnimGraphNode_StateMachine_0.FE_Pack
  ids=id
  nodes [
   AnimStateNode_0 = FE_Closed() {is=title AnimStateNode}
   AnimStateNode_2 = FE_Open_Into() {is=title AnimStateNode, stateEntered {guid=#1, notifyName=FE_OpenInto_Entered}}
   AnimStateNode_3 = FE_Open_Out() {is=title AnimStateNode}
   AnimStateTransitionNode_0 = AnimStateTransitionNode()
   AnimStateTransitionNode_3 = AnimStateTransitionNode() {bAutomaticRuleBasedOnSequencePlayerInState=true, crossfadeDuration=0.05}
   AnimStateTransitionNode_4 = AnimStateTransitionNode() {bAllowInertializationForSelfTransitions=true}
  ]
  links [
   {
    dst=FE_Open_Into
    rule=AnimGraph.AnimGraphNode_StateMachine_0.FE_Pack.AnimStateTransitionNode_0.Transition
    src=FE_Closed
    via=AnimStateTransitionNode_0
   }
   {
    dst=FE_Open_Out
    rule=AnimGraph.AnimGraphNode_StateMachine_0.FE_Pack.AnimStateTransitionNode_4.Transition
    src=FE_Open_Into
    via=AnimStateTransitionNode_4
   }
   {
    dst=FE_Closed
    rule=AnimGraph.AnimGraphNode_StateMachine_0.FE_Pack.AnimStateTransitionNode_3.Transition
    src=FE_Open_Out
    via=AnimStateTransitionNode_3
   }
  ]
 }
 FE_Closed {
  path=AnimGraph.AnimGraphNode_StateMachine_0.FE_Pack.AnimStateNode_0.FE_Closed
  nodes [Output Animation Pose(Result=Sequence Player 'Idle_Closed_PROP'() {sequence=/Game/Accessories/FORT_Backpacks/Backpack_F_MED_Scholar/Animation/Idle_Closed_PROP})]
 }
 FE_Open_Into {
  path=AnimGraph.AnimGraphNode_StateMachine_0.FE_Pack.AnimStateNode_2.FE_Open_Into
  nodes [
   n1 = Sequence Player 'WordUnity_Idle_Open_Into_PROP_FE'() {bLoopAnimation=false, sequence=/Game/Accessories/FORT_Backpacks/Backpack_F_MED_Scholar/Animation/WordUnity/WordUnity_Idle_Open_Into_PROP_FE}
   Output Animation Pose(Result=@n1)
  ]
 }
 FE_Open_Out {
  path=AnimGraph.AnimGraphNode_StateMachine_0.FE_Pack.AnimStateNode_3.FE_Open_Out
  nodes [
   n1 = Sequence Player 'WordUnity_Idle_Open_Out_PROP_FE'() {bLoopAnimation=false, sequence=/Game/Accessories/FORT_Backpacks/Backpack_F_MED_Scholar/Animation/WordUnity/WordUnity_Idle_Open_Out_PROP_FE}
   Output Animation Pose(Result=@n1)
  ]
 }
 Transition {
  path=AnimGraph.AnimGraphNode_StateMachine_0.FE_Pack.AnimStateTransitionNode_0.Transition
  nodes [Result(bCanEnterTransition=Get IsOpen() -> (IsOpen:bool))]
 }
 Transition {path=AnimGraph.AnimGraphNode_StateMachine_0.FE_Pack.AnimStateTransitionNode_3.Transition, nodes [Result()]}
 Transition {
  path=AnimGraph.AnimGraphNode_StateMachine_0.FE_Pack.AnimStateTransitionNode_4.Transition
  nodes [Result(bCanEnterTransition=Get FullyOpen() -> (FullyOpen:bool))]
 }
 Pack {
  entry=Closed
  path=AnimGraph.AnimGraphNode_StateMachine_2.Pack
  ids=id
  nodes [
   AnimStateNode_0 = Closed() {is=title AnimStateNode}
   AnimStateNode_2 = Open_Into() {is=title AnimStateNode}
   AnimStateNode_3 = Open_Out() {is=title AnimStateNode}
   AnimStateTransitionNode_0 = AnimStateTransitionNode()
   AnimStateTransitionNode_3 = AnimStateTransitionNode() {bAutomaticRuleBasedOnSequencePlayerInState=true, crossfadeDuration=0.05}
   AnimStateTransitionNode_4 = AnimStateTransitionNode() {bAllowInertializationForSelfTransitions=true}
  ]
  links [
   {
    dst=Open_Into
    rule=AnimGraph.AnimGraphNode_StateMachine_2.Pack.AnimStateTransitionNode_0.Transition
    src=Closed
    via=AnimStateTransitionNode_0
   }
   {
    dst=Open_Out
    rule=AnimGraph.AnimGraphNode_StateMachine_2.Pack.AnimStateTransitionNode_4.Transition
    src=Open_Into
    via=AnimStateTransitionNode_4
   }
   {
    dst=Closed
    rule=AnimGraph.AnimGraphNode_StateMachine_2.Pack.AnimStateTransitionNode_3.Transition
    src=Open_Out
    via=AnimStateTransitionNode_3
   }
  ]
 }
 Closed {
  path=AnimGraph.AnimGraphNode_StateMachine_2.Pack.AnimStateNode_0.Closed
  nodes [Output Animation Pose(Result=Sequence Player 'Idle_Closed_PROP'() {sequence=&Idle_Closed_PROP})]
 }
 Open_Into {
  path=AnimGraph.AnimGraphNode_StateMachine_2.Pack.AnimStateNode_2.Open_Into
  nodes [Output Animation Pose(Result=Sequence Player 'Idle_Open_Into_PROP'() {bLoopAnimation=false, sequence=/Game/Accessories/FORT_Backpacks/Backpack_F_MED_Scholar/Animation/Idle_Open_Into_PROP})]
 }
 Open_Out {
  path=AnimGraph.AnimGraphNode_StateMachine_2.Pack.AnimStateNode_3.Open_Out
  nodes [
   n1 = Sequence Player 'WordUnity_Idle_Open_Out_PROP'() {bLoopAnimation=false, sequence=/Game/Accessories/FORT_Backpacks/Backpack_F_MED_Scholar/Animation/WordUnity/WordUnity_Idle_Open_Out_PROP}
   Output Animation Pose(Result=@n1)
  ]
 }
 Transition {path=AnimGraph.AnimGraphNode_StateMachine_2.Pack.AnimStateTransitionNode_0.Transition, nodes [Result(bCanEnterTransition=Get IsOpen())]}
 Transition {path=AnimGraph.AnimGraphNode_StateMachine_2.Pack.AnimStateTransitionNode_3.Transition, nodes [Result()]}
 Transition {path=AnimGraph.AnimGraphNode_StateMachine_2.Pack.AnimStateTransitionNode_4.Transition, nodes [Result(bCanEnterTransition=Get FullyOpen())]}
 EventGraph {nodes [
   Event OnCharacterCustomizationLoaded()
   > n1 = Cast To FortPlayerPawn(Object=TryGetFortPawn())
   > n2 = Cast To B_Athena_PartModifier_Backpack_WordUnity!(
    execute
    Object=GetModifierForPartType!(self:Fort Player Pawn Object Reference=@n1, PartType:EFortCustomPartType Enum=Backpack) -> (ReturnValue:Custom Character Part Modifier Object Reference)
   ) -> (then, CastFailed, AsB Athena Part Modifier Backpack Word Unity:B Athena Part Modifier Backpack Word Unity Object Reference) {is=title K2Node_DynamicCast}
   > Branch(Condition=Get ReactivityEnabled!(self:B Athena Part Modifier Backpack Word Unity Object Reference=@n2) -> (ReactivityEnabled:bool))
   > Set IsOpen(IsOpen:bool=false) -> (Output_Get:bool)
   > Do Once()
   > n3 = Delay(Duration=4)
   Set IsOpen(execute=[@n7.then, @n3.then], IsOpen=true)
   > Delay(Duration=2.5)
   > Set IsOpen(IsOpen=false)
   > Set FullyOpen(FullyOpen:bool=true) -> (Output_Get:bool)
   > Delay(Duration=0.2)
   > Set FullyOpen(FullyOpen=false)
   AnimNotify_ReactivityCue() {is=title K2Node_Event}
   > Call Reactivity Cue!(execute:Exec, self:F MED Scholar Mimic Glean Pack Anim BP Object Reference) -> (then:Exec) {is=title K2Node_CallDelegate}
   > n4 = SpawnSystemAttached!(
    execute
    SystemTemplate:Niagara System Object Reference=/BRCosmetics/Accessories/FORT_Backpacks/Backpack_F_MED_WordUnity/FX/NS_Backpack_WordUnity
    AttachToComponent:Scene Component Object Reference=GetOwningComponent()
    AttachPointName:Name=pack
    Location:Vector="0, 0, 0"
    Rotation:Rotator="0, 0, 0"
    LocationType:EAttachLocation Enum=KeepRelativeOffset
    bAutoDestroy:bool=false
    bAutoActivate:bool=true
    PoolingMethod:ENCPoolMethod Enum=None
    bPreCullCheck:bool=true
   ) -> (then, ReturnValue:Niagara Particle System Component Object Reference)
   > Set OutlineFX(OutlineFX:Niagara Particle System Component Object Reference=@n4) -> (Output_Get:Niagara Particle System Component Object Reference)
   Event BlueprintPostEvaluateAnimation() {is=title K2Node_Event}
   > Set IsOpen(IsOpen=false)
   > Set FullyOpen(FullyOpen=false)
   AnimNotify_FXOff() {is=title K2Node_Event}
   n5 = Get OutlineFX() -> (OutlineFX:Niagara Particle System Component Object Reference)
   > "Set Niagara Variable (Bool)"!(execute, self:Niagara Particle System Component Object Reference=@n5, InVariableName:Name=bFinished, InValue:bool=true) -> (then) {is=title K2Node_CallFunction}
   > Deactivate(self:Actor Component Object Reference=@n5)
   n6 = Event BlueprintProcessAnimInputEvent()
   > Branch(Condition=IsNamed(self=@n6.AnimInputEvent, InEventName=aie.EnemyPlayerEliminated))
   > Set IsOpen(IsOpen=true)
   > Delay(Duration=2.5)
   > Set IsOpen(IsOpen=false)
   > Set FullyOpen(FullyOpen=true)
   > Delay(Duration=0.2)
   > Set FullyOpen(FullyOpen=false)
   AnimNotify_FE_OpenInto_Entered() {is=title K2Node_Event}
   > Set IsOpen(IsOpen=false)
   > n7 = Delay(Duration=14)
  ]}
 ReactivityCue {nodes [ReactivityCue() {is=title K2Node_FunctionEntry}]}
]
variables [
 FullyOpen {type=boolean}
 IsOpen {type=boolean}
 OutlineFX {type=NiagaraComponent, default=null}
 ReactivityCue {default=()}
]
